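; VehiclePropPicker is an actor Blueprint that places a shopping cart building prop where it sits when play begins.
; On BeginPlay, only the network authority spawns the hard-coded Athena shopping cart class at the Scene component's world transform, ignoring collisions and multiplying scale with the root.
; Despite the picker name, there is no selection logic, and the construction script is empty.
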(event EventBeginPlay
  (Networking|SwitchHasAuthority
    (:Authority
      (Game|SpawnActorfromClass "/Game/Athena/BuildingActors/Props/Building/ActorBlueprints/Containers/Athena_Prop_Commerce_ShoppingCart_01.Athena_Prop_Commerce_ShoppingCart_01_C" (Transformation|GetWorldTransform (Variables|Default|GetScene)) "AlwaysSpawn" "MultiplyWithRoot" 0 0 0 0))))

(fn ConstructionScript)
